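# Post-process AnimBlueprint for the Skull head skeleton that drives the jaw and shrinks part of the spine.
# It copies the attached parent mesh's pose and curves, then additively rolls C_jaw, weighted by the jaw_open_pose curve.
# A second bone modifier scales spine_04 down to a hundredth before output.
# Initialization sets Scorpion when the character lacks the CoffeePuff meta tag, but the output chain ignores it and an unconnected bool blend.
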AnimBlueprint /CoffeePuffGameplay/Assets/Meshes/Skull/Skull_Head_AnimBP_PostProcess
target_skeleton=&/Skull_Head_Skeleton
graphs [
 AnimGraph {nodes [
   Blend Poses by bool(BlendPose_1=Input Pose()) {blendType=Linear}
   n1 = "Transform (Modify) Bone - Bone: C_jaw"!(
    ComponentPose=Local To Component(LocalPose=Copy Pose From Mesh() {bCopyCurves=true, bUseAttachedParent=true})
    AlphaCurveName:Name=jaw_open_pose
    Translation="0,0,0"
    Rotation="0,0,25"
    Scale="1,1,1"
   ) -> (Pose) {
    alphaInputType=Curve
    rotationMode=BMM_Additive
    rotationSpace=BCS_BoneSpace
    alphaBoolBlend {blendOption=Linear}
    boneToModify {boneName=C_jaw}
    rotation {roll=25}
   }
   n2 = "Transform (Modify) Bone - Bone: spine_04"(ComponentPose=@n1, Alpha=1, Translation="0,0,0", Rotation="0,0,0", Scale="0.01,0.01,0.01") {
    scaleMode=BMM_Additive
    scaleSpace=BCS_BoneSpace
    alphaBoolBlend {blendOption=Linear}
    boneToModify {boneName=spine_04}
    scale=<0.01 0.01 0.01>
   }
   Output Pose(Result=Component To Local(ComponentPose=@n2))
  ]}
 EventGraph {nodes [
   Event BlueprintInitializeAnimation()
   n1 = Make Array("[0]"=(TagName="Cosmetics.MetaTags.CoffeePuff"))
   > Set Scorpion(Scorpion:bool=NOT Boolean(A=CharacterItemDefinitionHasAnyMetaTag(Tags=@n1))) -> (Output_Get:bool)
   Event BlueprintUpdateAnimation()
  ]}
]
variables [Scorpion {type=boolean}]
